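; Hologram screen component for the Conditional Button V2 device that displays a required key item and its lock state.
; It binds to a key lock component and the owning device to react to item, key count, lock, inventory and play mode changes.
; Item changes async-load the preview icon and cache name and rarity for missing-key text.
; Cosmetic-gated updates drive a dynamic material's textures, colors and time-stamped animations, tinting by whether the local player holds the keys.
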(event Custom|Load%20New%20Item
  (params (ItemDef :type Item%20Variant%20Handle%20Structure))
  (bind _itemdefintion (Creative|ItemData|GetItemDefinition ItemDef "/Script/FortniteGame.FortWorldItemDefinition"))
  (if (not (Editor|InEditor))
    (Utilities|FlowControl|Delay (Components|Timeline|GetTimelineLength (Variables|/CreativeCoreDevices/Device_ConditionalButton_V2.Device_ConditionalButton_V2_C|GetDragItem (CallFunction|GetLockParent))))
    (else
      _))
  (Utilities|IsValid _itemdefintion
    (:Is%20Valid
      (bind _output_get (Variables|Default|SetItemDef ItemDef))
      (Variables|Default|SetDisplay%20Name (Creative|ItemData|GetDisplayName _output_get))
      (bind _returnvalue (Fortnite|GetRarityDataForItemDef _itemdefintion))
      (Variables|Default|SetRarity%20Name (Utilities|Struct|BreakFortRarityItemData _returnvalue))
      (Networking|CanExecuteCosmeticEvents
        (:True
          (bind _loadedobject (Utilities|LoadAsset (select (Utilities|IsValidSoftObjectReference (Item|GetLargePreviewImage _itemdefintion)) (Item|GetLargePreviewImage _itemdefintion) (Item|GetSmallPreviewImage _itemdefintion)))
            (:Loaded
              (bind _astexture (Utilities|Casting|CastToTexture _loadedobject))
              (Variables|Default|SetKeyItemTexture _astexture))))))
    (:Is%20Not%20Valid
      ;[1 1 1] If failed item def then no item def. We'll want to update to no key.
      (Variables|Default|SetItemDef)
      (Variables|Default|SetKeyItemTexture "")
      (Networking|CanExecuteCosmeticEvents
        (:True
          _))))
  (SetScreenElement|UpdateScreenKey)
  (Default|CallOnKeyTextureUpdated)
  (if (not (Editor|InEditor))
    (Audio|PlaySoundatLocation (Variables|Default|GetUpdate%20Item%20Sound) (Transformation|GetActorLocation (Components|GetOwner)))))

(event Custom|Play%20Relock
  (Networking|CanExecuteCosmeticEvents
    (:True
      (Variables|Default|SetIsLocked true)
      (SetScreenElement|SetConstantInteract false)
      (CallFunction|UpdateScreenColorfromLocalPawn))))

(event Custom|Play%20Unlock
  (Networking|CanExecuteCosmeticEvents
    (:True
      (Variables|Default|SetIsLocked false)
      (SetScreenElement|SetConstantInteract true)
      (CallFunction|UpdateScreenColorfromLocalPawn)
      (PlayAnimations|PlayInteraction true)
      (Utilities|FlowControl|Delay (+ (Variables|/CreativeCoreDevices/Device_ConditionalButton_V2.Device_ConditionalButton_V2_C|GetUseKeyAnimationLength (CallFunction|GetLockParent)) 0.2))
      (Default|CallOnKeyUnlockComplete))))

(event Custom|PlayModeChanged
  (params (Play%20Mode :type Boolean))
  (CallFunction|UpdateScreenEdgeForSelection (and (not Play%20Mode) (== (Variables|/CreativeCoreDevices/Device_ConditionalButton_V2.Device_ConditionalButton_V2_C|GetNextItemToAdd (CallFunction|GetLockParent)) (Variables|Default|GetHologramNum))))
  (CallFunction|PlayRelock))

(event Custom|Initialize%20Component
  (SetScreenElement|UpdateScreenKey)
  (CallFunction|UpdateScreenColorfromLocalPawn))

(fn BindToEvents
  (params (Key%20Lock :type BP%20Key%20Lock%20Component%20Object%20Reference))
  (bind _asbp_lock_device (CallFunction|GetLockParent))
  (bind _key_lock (Variables|Default|GetKey%20Lock))
  (bind _output_get (Variables|Default|SetKey%20Lock Key%20Lock))
  (EventDispatchers|BindEventtoOnKeyDefinitionUpdated _output_get (EventDispatchers|CreateEvent :!event Load%20New%20Item))
  (Networking|CanExecuteCosmeticEvents
    (:True
      (EventDispatchers|BindEventtoOnRemainingKeysUpdated _key_lock (EventDispatchers|CreateEvent :!event Update%20Keys%20Remaining))
      (EventDispatchers|BindEventtoOnUnlocked _key_lock (EventDispatchers|CreateEvent :!event Play%20Unlock))
      (EventDispatchers|BindEventtoOnLocked _key_lock (EventDispatchers|CreateEvent :!event Play%20Relock))
      (Default|BindEventtoOnPlayModeChanged _asbp_lock_device (EventDispatchers|CreateEvent :!event PlayModeChanged))
      (EventDispatchers|BindEventtoOnLocalPawnInventoryChanged _asbp_lock_device (EventDispatchers|CreateEvent :!event Update%20Screen%20Color%20From%20Local%20Pawn)))))

(fn GetLockParent
  (bind _asdevice_conditional_button_v2 (Utilities|Casting|CastToDevice_ConditionalButton_V2 (Components|GetOwner)))
  (return _asdevice_conditional_button_v2))

(fn GetMID
  (Variables|Default|GetMID
    (:then
      _)
    (:else
      (Variables|Default|SetMID (Material|GetMIDfromMaterial self 0 (Variables|Default|GetMaterial)))))
  (return (Variables|Default|GetMID)))

(fn GetMissingText
  (params (Add%20Rarity :type Boolean))
  (bind _returnvalue (Fortnite|GetFirstLocalFortPlayerController true))
  (bind _returnvalue_2 (Game|Player|GetPlayerPawn _returnvalue))
  (if (Variables|Default|GetIsLocked)
    (Utilities|IsValid _returnvalue_2
      (:Is%20Valid
        (bind _key_lock (Variables|Default|GetKey%20Lock)
          (:then
            (bind _returnvalue_3 (Class|FortCreativeKeyLockComponent|BlueprintHasKeystoInteract _key_lock _returnvalue_2))
            (LocalVariables|SetMissingItems (not _returnvalue_3)))
          (:else))))
    (else
      (LocalVariables|SetMissingItems false)))
  (return (LocalVariables|GetMissingItems) (select (LocalVariables|GetMissingItems) (Utilities|Text|FormatText (select (Variables|GetAdd%20Rarity) "NSLOCTEXT(\"[529FE6B6438CDD473B5E81AE202CAFB0]\", \"131D383C448459EB5A0D939D31C56A85\", \"{Name} ({Rarity})\")" "NSLOCTEXT(\"[529FE6B6438CDD473B5E81AE202CAFB0]\", \"3835407543D46B5E7E23C1A3A601BDFC\", \"{Name}\")") (Variables|Default|GetRarity%20Name) (Variables|Default|GetDisplay%20Name)) "")))

(fn PlayConsumeKeyAnim
  (bind _mid (CallFunction|GetMID))
  (Rendering|Material|SetScalarParameterValue _mid "UseKeyAnimationLength" (Variables|/CreativeCoreDevices/Device_ConditionalButton_V2.Device_ConditionalButton_V2_C|GetUseKeyAnimationLength (CallFunction|GetLockParent)) :!class /Script/Engine.MaterialInstanceDynamic)
  (Rendering|Material|SetScalarParameterValue _mid "UseKeyAnimationTime" (Utilities|Time|GetGameTimeinSeconds) :!class /Script/Engine.MaterialInstanceDynamic)
  (Rendering|Material|SetScalarParameterValue _mid "UseKeyAnimationLoops" (Math|Conversions|ToFloat(Integer) (Variables|Default|GetCached%20Key%20Number)) :!class /Script/Engine.MaterialInstanceDynamic))

(fn PlayInteraction
  (params (Success :type Boolean :default true))
  (bind _mid (CallFunction|GetMID))
  (Rendering|Material|SetScalarParameterValue _mid "Interaction Animation Time" (Utilities|Time|GetGameTimeinSeconds) :!class /Script/Engine.MaterialInstanceDynamic)
  (Rendering|Material|SetScalarParameterValue _mid "User Selected Reset Delay" (Variables|/CreativeCoreDevices/Device_ConditionalButton_V2.Device_ConditionalButton_V2_C|GetResetDelay (CallFunction|GetLockParent)) :!class /Script/Engine.MaterialInstanceDynamic)
  (Rendering|Material|SetScalarParameterValue _mid "Successful Unlock" (Math|Conversions|ToFloat(Boolean) (not (Variables|Default|GetIsLocked))) :!class /Script/Engine.MaterialInstanceDynamic)
  (Rendering|Material|SetScalarParameterValue _mid "Interaction Type" (Math|Conversions|ToFloat(Boolean) (or Success (not (Variables|Default|GetIsLocked)))) :!class /Script/Engine.MaterialInstanceDynamic))

(fn SetCanInteract
  (params (Yes :type Boolean))
  (SetScreenElement|SetScreenColor (select (or Yes (not (Variables|Default|GetIsLocked))) "NewEnumerator1" "NewEnumerator0")))

(fn SetIcon
  (params (Primary%20Icon :type Boolean) (Icon :type Texture%20Object%20Reference))
  (Networking|CanExecuteCosmeticEvents
    (:True
      (bind _mid (CallFunction|GetMID))
      (Rendering|Material|SetTextureParameterValue _mid (Utilities|String|StringToName (Utilities|String|SelectString "State A Icon" "State B Icon" Primary%20Icon)) Icon))))

(fn SetScreenColor
  (params (Valid :type ELockDevicePreviewColor%20Enum :default "NewEnumerator0"))
  (bind (_icon _edge _background) (Class|DeviceConditionalButtonV2|ValidToColor (CallFunction|GetLockParent) Valid))
  (bind _mid (CallFunction|GetMID))
  (Rendering|Material|SetVectorParameterValue _mid "Icon Color" _icon :!class /Script/Engine.MaterialInstanceDynamic)
  (Rendering|Material|SetVectorParameterValue _mid "Edge Color" _edge :!class /Script/Engine.MaterialInstanceDynamic)
  (Rendering|Material|SetVectorParameterValue _mid "Background Color" _background :!class /Script/Engine.MaterialInstanceDynamic))

(fn SetConstantInteract
  (params (%54rue :type Boolean))
  (bind _mid (CallFunction|GetMID))
  (Rendering|Material|SetScalarParameterValue _mid "ConstantInteract" (Math|Float|SelectFloat 1.0 0.0 %54rue) :!class /Script/Engine.MaterialInstanceDynamic))

(fn ShowKeysRequired
  (params (Show :type Boolean))
  (bind _mid (CallFunction|GetMID))
  (Rendering|Material|SetScalarParameterValue _mid "Show Number of Keys" (Math|Conversions|ToFloat(Boolean) Show) :!class /Script/Engine.MaterialInstanceDynamic))

(fn UpdateKeysRemaining
  (params (New%20Number :type Integer))
  (bind _mid (CallFunction|GetMID))
  (Rendering|Material|SetScalarParameterValue _mid "Number of Keys Required" (Math|Conversions|ToFloat(Integer) New%20Number) :!class /Script/Engine.MaterialInstanceDynamic)
  (if (> New%20Number 0)
    (Variables|Default|SetCached%20Key%20Number (Variables|GetNew%20Number))))

(fn UpdateScreenColorFromLocalPawn
  (bind _asbp_lock_device (CallFunction|GetLockParent))
  (bind _returnvalue (Fortnite|GetFirstLocalFortPlayerController true))
  (bind _returnvalue_2 (Pawn|GetControlledPawn _returnvalue))
  (bind _bisminigameinprogress (Variables|/CreativeCoreDevices/Device_ConditionalButton_V2.Device_ConditionalButton_V2_C|GetbIsMinigameInProgress _asbp_lock_device))
  (bind _returnvalue_3 (not _bisminigameinprogress))
  (Networking|CanExecuteCosmeticEvents
    (:True
      (Utilities|IsValid _returnvalue
        (:Is%20Valid
          (Utilities|IsValid _returnvalue_2
            (:Is%20Valid
              (bind _returnvalue_4 (Class|FortCreativeKeyLockComponent|BlueprintHasKeystoInteract (Variables|Default|GetKey%20Lock) _returnvalue_2))
              (CallFunction|SetCanInteract (or _returnvalue_4 _returnvalue_3))
              (Development|PrintString (Utilities|String|Append (Utilities|String|ToString(Text) (Variables|Default|GetDisplay%20Name)) "" " Interact = " (Utilities|String|ToString(Boolean) (or _returnvalue_4 _returnvalue_3))) (Variables|Default|GetShowDebug) (Variables|Default|GetShowDebug)))
            (:Is%20Not%20Valid
              _)))
        (:Is%20Not%20Valid
          _))
      (CallFunction|SetCanInteract _returnvalue_3))))

(fn UpdateScreenEdgeforSelection
  (params (Selection :type Boolean))
  (bind _mid (CallFunction|GetMID))
  (Rendering|Material|SetScalarParameterValue _mid "EdgeMultiplier" (select Selection (Variables|Default|GetItemAdd%20Edge%20Multiplier) 1.0) :!class /Script/Engine.MaterialInstanceDynamic))

(fn UpdateScreenKey
  (Networking|CanExecuteCosmeticEvents
    (:True
      (LocalVariables|SetValid%20Item (Utilities|IsValid (Variables|Default|GetKeyItemTexture) :!class /Script/Engine.KismetSystemLibrary))
      (bind _mid (CallFunction|GetMID))
      (Rendering|Material|SetTextureParameterValue _mid "Key Texture" (select (and (LocalVariables|GetValid%20Item) (or (Utilities|Enum|Equal(Enum) (Variables|/CreativeCoreDevices/Device_ConditionalButton_V2.Device_ConditionalButton_V2_C|GetShowKeyItem (CallFunction|GetLockParent)) "NewEnumerator1") (Utilities|Enum|Equal(Enum) (Variables|/CreativeCoreDevices/Device_ConditionalButton_V2.Device_ConditionalButton_V2_C|GetShowKeyItem (CallFunction|GetLockParent)) "NewEnumerator2"))) (Variables|Default|GetKeyItemTexture) (Variables|Default|GetUnknown%20Key%20Texture)))
      (Rendering|Material|SetVectorParameterValue _mid "Key color multiplier" (Math|Color|SelectColor "(R=1.000000,G=1.000000,B=1.000000,A=1.000000)" (Variables|Color|GetUnknown%20Key%20Color) (LocalVariables|GetValid%20Item)) :!class /Script/Engine.MaterialInstanceDynamic)
      (Rendering|Material|SetScalarParameterValue _mid "UseTransition" (Utilities|Select :Index (Variables|/CreativeCoreDevices/Device_ConditionalButton_V2.Device_ConditionalButton_V2_C|GetShowKeyItem (CallFunction|GetLockParent)) :NewEnumerator0 0.0 :NewEnumerator1 1.0 :NewEnumerator2 0.0 :NewEnumerator3 -1.0 :NewEnumerator4 1.0) :!class /Script/Engine.MaterialInstanceDynamic))))

(fn UpdateSpawnAnimation
  (params (NewTime :type Float%20%28double-precision%29))
  (bind _mid (CallFunction|GetMID))
  (Rendering|Material|SetScalarParameterValue _mid "Spawn Animation Time" NewTime :!class /Script/Engine.MaterialInstanceDynamic))
